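# Parent metal material for the DarkNexus set, assigned the Metal physical material and enabled for Nanite, instanced static and skeletal meshes.
# It uses material attributes, taking its base surface from MF_DarkNexus_Metal.
# The UseValetDamage static switch optionally layers the Valet vehicle damage function over that surface.
# Either result then feeds MF_FN_DestructibleDamage, so destructible damage is applied last in both variants.
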Material /Hera/Sets/DarkNexus/Materials/Parents/M_DarkNexus_Metal_Parent
graph {
 properties {
  bAutomaticallySetUsageInEditor=false
  bUseMaterialAttributes=true
  bUsedWithInstancedStaticMeshes=true
  bUsedWithMeshDeformer=true
  bUsedWithNanite=true
  bUsedWithSkeletalMesh=true
  ditheredLODTransition=true
  physMaterial=/Game/FortressPhysicalMaterials/Metal
  pixelDepthOffsetMode=PDOM_Legacy
  usedShadingModels=Default Lit
 }
 nodes [
  n1 = MaterialFunctionCall() {materialFunction=/Hera/Sets/DarkNexus/MaterialFunctions/MF_DarkNexus_Metal}
  MaterialFunctionCall(Material_In=StaticSwitchParameter(
    True=MaterialFunctionCall!(In=@n1.Result, Metal Tint, Local Bounds Split) {materialFunction=/Game/Vehicles/Valet/MF_Valet_Damage}.Result
    False=@n1.Result
   ) {parameterName=UseValetDamage}) {materialFunction=&MF_FN_DestructibleDamage}
 ]
}
